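# Prototype test material, named as a red light, for the Test Chamber game in the Battle Pass S22 narrative quest actors.
# It uses the Default Lit shading model and holds a single constant red color node.
# The graph shows no textures, parameters or math, and no link from the color to a material output.
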Material /BattlePassS22/QuestActors/Narrative_Pt1/TestChamber_Game/PrototypeMats/Test_Mat_RedLight
graph {properties {pixelDepthOffsetMode=PDOM_Legacy, usedShadingModels=Default Lit}, nodes [Constant3Vector() {constant=<rgba 0.5469 0 0.0361 0>}]}
